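; Gameplay ability for the ColdFern weapon that toggles its reload override based on magazine ammo count.
; On activation, if a scalable flag enables it, it caches the avatar's current weapon and that weapon's ColdFern override-reload component.
; It evaluates ammo immediately, then again on each local ammo change event.
; The override is on above a scalable threshold, and the component is updated only when that state changes.
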(event Ability|EventActivateAbility
  (bind _player_weapon (Variables|Default|GetPlayer%20Weapon))
  ;[0 1 0] Get player pawn and weapon
  (if (ScalableFloat|GetValueasBool (Variables|Default|GetEnableReloadOverride))
    (bind _asfort_player_pawn (Utilities|Casting|CastToFortPlayerPawn (Ability|GetAvatarActorfromActorInfo)))
    (bind _output_get (Variables|Default|SetPlayerPawn _asfort_player_pawn))
    (bind _output_get_2 (Variables|Default|SetPlayer%20Weapon (Variables|/Script/FortniteGame.FortPawn|GetCurrentWeapon _output_get)))
    (Utilities|IsValid _output_get_2
      (:Is%20Valid
        (bind _output_get_3 (Variables|Default|SetColdFernComponent (Actor|GetComponentbyClass _output_get_2 "/WildEstateWeapons/Gameplay/ColdFern/Components/FortWeaponComponent_OverrideReload_ColdFern.FortWeaponComponent_OverrideReload_ColdFern_C")))
        (Utilities|IsValid _output_get_3
          (:Is%20Valid
            ; Check ammo count and update fire mode
            (CallFunction|UpdateOverrideReloadbyAmmoCount (FortWeapon|GetLocalMagazineAmmoCount _player_weapon) (ScalableFloat|GetValueasInteger (Variables|Default|GetOverrideWhenAmmoAboveAmount)))
            ; Bind to future ammo changes
            (EventDispatchers|BindEventtoOnLocalAmmoChanged _player_weapon (EventDispatchers|CreateEvent :!event Custom|OnMagazineAmmoCountChanged))))))))

(event Custom|OnMagazineAmmoCountChanged
  (params (LocalCount :type Integer) (LocalRemaining :type Integer))
  (CallFunction|UpdateOverrideReloadbyAmmoCount LocalCount (ScalableFloat|GetValueasInteger (Variables|Default|GetOverrideWhenAmmoAboveAmount))))

(fn UpdateOverrideReloadByAmmoCount
  (params (AmmoCount :type Integer) (OverrideWhenAboveAmount :type Integer))
  (bind _output_get (Variables|Default|SetShouldOverride (> AmmoCount OverrideWhenAboveAmount)))
  (if (!= (Variables|Default|GetLastUpdatedOverride) _output_get)
    (Variables|Default|SetLastUpdatedOverride _output_get)
    (bind _coldferncomponent (Variables|Default|GetColdFernComponent)
      (:then
        (Class|FortWeaponComponentOverrideReloadColdFern|SetOverrideReload _coldferncomponent (Variables|Default|GetShouldOverride)))
      (:else))))
